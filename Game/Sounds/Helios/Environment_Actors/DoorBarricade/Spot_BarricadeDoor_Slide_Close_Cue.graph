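# Sound cue for a barricade door sliding closed, using the shared door attenuation and sound class.
# A random node picks between two equally weighted wave variations and feeds a modulator, which serves all four inputs of the quality-level root.
# It sends to the default reverb submix and an exterior ambient source bus, and overrides concurrency to prevent new instances once the limit is reached.
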SoundCue /Game/Sounds/Helios/Environment_Actors/DoorBarricade/Spot_BarricadeDoor_Slide_Close_Cue
root=QualityLevel_0
nodes [
 {id=Modulator_2, children [Random_0]}
 {id=QualityLevel_0, children [Modulator_2,Modulator_2,Modulator_2,Modulator_2]}
 {id=Random_0, children [WavePlayer_1,WavePlayer_2], properties {weights [1,1]}}
 {id=WavePlayer_1, properties {soundWaveAssetPtr=&/Spot_Door_Barricade_Sliding_Close_01}}
 {id=WavePlayer_2, properties {soundWaveAssetPtr=&/Spot_Door_Barricade_Sliding_Close_02}}
]
properties {
 attenuationSettings=/Game/Sounds/Attenuation/Fort_Doors_Att
 bOverrideConcurrency=true
 maxDistance=2400
 soundClassObject=&Fort_Doors
 volumeMultiplier=1
 busSends [{audioBus=null, minSendLevel=0.8, sendLevel=1, soundSourceBus=&AmbientAudio_Exterior_SourceBus}]
 concurrencyOverrides {resolutionRule=PreventNew}
 soundSubmixSends [{minSendLevel=0.6, soundSubmix=&Fortnite_Reverb_Default}]
}
